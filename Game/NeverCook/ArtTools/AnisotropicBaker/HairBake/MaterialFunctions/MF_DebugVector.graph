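# MF_DebugVector is a material function in the AnisotropicBaker hair bake tools that visualizes a direction vector as shading on tiled spheres.
# It scales the texture coordinates by a Tiling input, defaulting to 32, and samples the engine's example sphere normal map.
# The output is the dot product of that normal and the Vector input, defaulting to +X, so each sphere is brightest where it faces the vector.
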MaterialFunction /Game/NeverCook/ArtTools/AnisotropicBaker/HairBake/MaterialFunctions/MF_DebugVector
graph {nodes [
  n1 = FunctionInput() {bUsePreviewValueAsDefault=true, inputName=Tiling, inputType=FunctionInput_Scalar, previewValue=<32 0 0 1>}
  n2 = TextureSample(UVs=Multiply(A=TextureCoordinate(), B=@n1)) {
   samplerSource=SSM_Wrap_WorldGroupSettings
   samplerType=SAMPLERTYPE_Normal
   texture=/Engine/Functions/Engine_MaterialFunctions02/ExampleContent/Textures/SphereRenderNormalsMap
  }
  n3 = Constant3Vector() {constant=<rgba 1 0 0 0>}
  FunctionOutput(None=DotProduct(A=@n2.RGB, B=FunctionInput(Preview=@n3) {bUsePreviewValueAsDefault=true, inputName=Vector}))
 ]}
