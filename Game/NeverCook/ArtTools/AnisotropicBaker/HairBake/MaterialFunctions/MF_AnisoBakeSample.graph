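# Material function for the anisotropic hair bake art tool that takes one offset tap of a texture and turns its change from the original sample into a 2D direction vector.
# It samples at UV plus Offset on a chosen MIP, dots both samples with the Channel mask, then scales Offset by the difference.
# The Method02 parameter blends in a sign-folded vector, so opposite directions land in the positive-X half-plane.
# Raw offset and offset sample are also output.
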MaterialFunction /Game/NeverCook/ArtTools/AnisotropicBaker/HairBake/MaterialFunctions/MF_AnisoBakeSample
graph {nodes [
  n1 = FunctionInput() {bUsePreviewValueAsDefault=true, inputName=Offset, inputType=FunctionInput_Vector2}
  n2 = FunctionInput(Preview=TextureSample() {texture=/Game/NeverCook/ArtTools/AnisotropicBaker/TestImages/T_TangentDebu}.RGB) {inputName=OriginalSample}
  n3 = FunctionInput() {bUsePreviewValueAsDefault=true, inputName=Channel, inputType=FunctionInput_Vector4, previewValue=<0 1 0 1>}
  n4 = ComponentMask(None=@n3) {b=true, g=true, r=true}
  n5 = FunctionInput(Preview=TextureCoordinate()) {bUsePreviewValueAsDefault=true, inputName=UV, inputType=FunctionInput_Vector2}
  n6 = TextureObject() {texture=&T_TangentDebu}
  n7 = DotProduct(
   A=TextureSample!(
    UVs=Add(A=@n1, B=@n5)
    Tex=FunctionInput(Preview=@n6) {inputName=Texture, inputType=FunctionInput_Texture2D}
    Level=FunctionInput() {bUsePreviewValueAsDefault=true, inputName=MIP, inputType=FunctionInput_Scalar}
    Apply View MipBias
   ) {mipValueMode=TMVM_MipLevel}.RGBA
   B=@n3
  )
  n8 = Multiply(
   A=FunctionOutput(None=@n1) {outputName=Direction}
   B=Subtract(A=DotProduct(A=@n2, B=@n4), B=FunctionOutput(None=@n7) {outputName=OffsetSample})
  )
  n9 = AppendVector(A=@n8, B=Constant())
  n10 = Constant3Vector() {constant=<rgba -1 -1 1 0>}
  n11 = MaterialFunctionCall(Float3=@n9) {materialFunction=&BreakOutFloat3Components}
  n12 = Sign(None=@n11.R)
  n13 = Multiply(A=@n12, B=Sign(None=@n11.G))
  n14 = LinearInterpolate(A=LinearInterpolate(A=Multiply(A=@n9, B=@n10), B=@n9, Alpha=Saturate(None=@n12)), B=Abs(None=@n9), Alpha=Saturate(None=Sign(None=@n13)))
  FunctionOutput(None=LinearInterpolate(A=@n8, B=ComponentMask(None=@n14) {g=true, r=true}, Alpha=ScalarParameter() {parameterName=Method02})) {outputName=AbsDifference}
  Normalize()
 ]}
